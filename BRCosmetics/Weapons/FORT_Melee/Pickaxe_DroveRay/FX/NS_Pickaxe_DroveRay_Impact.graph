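# Niagara impact effect for the DroveRay pickaxe, built from burst emitters for a scaling mesh with a light, sparks, debris, smoke and glow.
# User.VariantID picks among four color schemes through nested bool selects and indexed arrays.
# User.HitWeakSpot gates an extra sprite whose sub-image follows the variant.
# Front-end preview parameters rescale sizes and velocities, while distance culling thins the particle bursts.
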NiagaraSystem /BRCosmetics/Weapons/FORT_Melee/Pickaxe_DroveRay/FX/NS_Pickaxe_DroveRay_Impact
dependencies {
 usedDataInterfaces [
  &NiagaraDataInterfaceArrayColor
  &NiagaraDataInterfaceCurve
  &NiagaraDataInterfaceColorCurve
  &NiagaraDataInterfaceMeshRendererInfo
  &NiagaraDataInterfaceVector2DCurve
  &NiagaraDataInterfacePlatformSet
  &NiagaraDataInterfaceSpriteRendererInfo
  &NiagaraDataInterfaceArrayFloat
 ]
 usedDynamicInputs [&RandomBool]
 usedModules [&SystemState]
}
emitters {
 Empty {
  emitterUpdateScript [
   &EmitterState()
   &SpawnBurst_Instantaneous(Spawn Count=1, Use Spawn Probability=false)
  ]
  particleSpawnScript [
   &InitializeParticle(
    Lifetime=0.5
    Mesh Scale Mode=::NewEnumerator3
    Mesh Uniform Scale=&FN_FrontEndFloatModifier(Gameplay=0.6, Front End=&Multiply_Float(A=0.7, B=$User.FrontEndSize))
    UsePositionOffset=false
    Write Material Random=true
   ) {module=InitializeParticle001}
   &InitialMeshOrientation(Orientation Axis=<0 0 1>, Orientation Vector=<0 0 1>, Rotation=<0 0.25 0>, Use Orientation Vector=true, Use Rotation Vector=true)
   &InitialMeshOrientation(
    Orientation Axis=<0 0 0>
    Orientation Vector=<0 0 0>
    Rotation=&RandomRangeVector(Minimum=<0 0 0>, Maximum=<1 0 0>)
    Use Orientation Vector=false
    Use Rotation Vector=true
   ) {module=InitialMeshOrientation001}
  ]
  particleUpdateScript [
   &DynamicMaterialParameters(
    Index 0 Param 1=&FloatFromCurve(FloatCurve=DataInterface() {propertyValues {Curve {keys [{arriveTangent=-2.1015, interpMode=RCIM_Cubic, leaveTangent=-2.1015, tangentMode=RCTM_User, time=0, value=0.5}, <at 1 -0.15>]}}})
    Index 0 Param 2=&FloatFromCurve(FloatCurve=DataInterface() {propertyValues {Curve {keys [{arriveTangent=-0.3464, interpMode=RCIM_Cubic, leaveTangent=-0.3464, tangentMode=RCTM_User, time=0, value=1}, <at 1 0.9>]}}})
    Index 0 Param 3=&FloatFromCurve(FloatCurve=DataInterface() {propertyValues {Curve {keys [<at 0 0>,<at 0.15 1>,<at 1 1.2>]}}})
    Module.Param0WriteEnabled=true
    Module.Param1WriteEnabled=true
    Module.Param2WriteEnabled=true
    Module.Param3WriteEnabled=false
   )
   &ScaleColor(
    Scale Alpha=&FloatFromCurve(FloatCurve=DataInterface() {propertyValues {Curve {keys [<at 0 1>,<at 1 0>]}}})
    Scale RGB=&VectorFromFloat(Value=&FloatFromCurve(FloatCurve=DataInterface() {propertyValues {Curve {keys [<at 0 10>,<at 0.15 3>,<at 1 1>]}}}))
    ScaleA=true
    ScaleRGB=true
   ) {enabled=false, module=ScaleColor001}
   &ScaleColor(
    Scale Mode=::NewEnumerator1
    Scale RGBA=&MakeVector4_FromLinearColor(LinearColor=&MakeCustomLinearColorFromBool(
      Bool=&SetBoolByFloatComparison(A=$User.VariantID, B=0, Comparison Type=::NewEnumerator4)
      True Color=&ColorFromCurve(ColorCurve=DataInterface() {propertyValues {
         AlphaCurve {keys [<at 0.1767 1.3>,<at 1 0>]}
         BlueCurve {keys [<at 0.0198 0>,<at 0.1952 0.0015>,<at 0.4571 0>,<at 0.7737 0.0106>]}
         GreenCurve {keys [<at 0.0198 1.339>,<at 0.1952 0.4199>,<at 0.4571 0.0181>,<at 0.7737 0>]}
         RedCurve {keys [<at 0.0198 2>,<at 0.1952 2>,<at 0.4571 0.5781>,<at 0.7737 0.2101>]}
        }})
      False Color=&MakeCustomLinearColorFromBool(
       Bool=&SetBoolByFloatComparison(A=$User.VariantID, B=1, Comparison Type=::NewEnumerator4)
       True Color=&ColorFromCurve(ColorCurve=DataInterface() {propertyValues {
          AlphaCurve {keys [<at 0.2215 1.3>,<at 1 0>]}
          BlueCurve {keys [<at 0.0198 0.2871>,<at 0.2838 0.0562>,<at 0.5671 0.0007>]}
          GreenCurve {keys [<at 0.0198 2>,<at 0.2838 2>,<at 0.5671 0.0885>]}
          RedCurve {keys [<at 0.0198 1.0241>,<at 0.2838 0.4624>,<at 0.5671 0.0168>]}
         }})
       False Color=&MakeCustomLinearColorFromBool(
        Bool=&SetBoolByFloatComparison(A=$User.VariantID, B=2, Comparison Type=::NewEnumerator4)
        True Color=&ColorFromCurve(ColorCurve=DataInterface() {propertyValues {
           AlphaCurve {keys [<at 0 1 RCIM_Cubic>,<at 1 0 RCIM_Cubic>]}
           BlueCurve {keys [<at 0.0259 9.9073 RCIM_Cubic>,<at 0.1799 1>,<at 0.3444 1>,<at 0.6432 0.1441 RCIM_Cubic>]}
           GreenCurve {keys [{arriveTangent=3.0456, interpMode=RCIM_Cubic, time=0.0259, value=12}, <at 0.1799 0.8276>, <at 0.3444 0.2181>, <at 0.6432 0.0067 RCIM_Cubic>]}
           RedCurve {keys [{arriveTangent=0.8921, interpMode=RCIM_Cubic, time=0.0259, value=0.5882}, <at 0.1799 0.1605>, <at 0.3444 0>, <at 0.6432 0.0099 RCIM_Cubic>]}
          }})
        False Color=&ColorFromCurve(ColorCurve=DataInterface() {propertyValues {
           AlphaCurve {keys [<at 0 1 RCIM_Cubic>,<at 1 0 RCIM_Cubic>]}
           BlueCurve {keys [{arriveTangent=0.5348, interpMode=RCIM_Cubic, time=0, value=0}, <at 0.1799 0.0962>, <at 0.3853 0>, <at 0.6662 0.0027 RCIM_Cubic>]}
           GreenCurve {keys [{arriveTangent=-0.1127, interpMode=RCIM_Cubic, time=0, value=8.6231}, <at 0.1799 0.8598>, <at 0.3853 0.4863>, <at 0.6662 0.0344 RCIM_Cubic>]}
           RedCurve {keys [<at 0 12 RCIM_Cubic>,<at 0.1799 1>,<at 0.3853 1>,<at 0.6662 0.0434 RCIM_Cubic>]}
          }})
       )
      )
     ))
    ScaleRGBA=true
   )
   &ParticleState()
   &ScaleMeshSize(Scale Factor=&VectorFromFloat(Value=&FloatFromCurve(
      FloatCurve=DataInterface() {propertyValues {Curve {keys [<at 0 0>,<at 0.1 1.4>,<at 1 2>]}}}
      Scale Curve=&Multiply_Float(A=1, B=1)
     )))
   &CameraOffset(Camera Offset Amount=10)
   &FN_Light_Attributes(
    Color=$Particles.Color
    Create Color Attribute=true
    Create Exponent Attribute=true
    Create Radius Attribute=true
    Exponent=3.5
    Radius=85
    Scale Color=&FloatFromCurve(
     FloatCurve=DataInterface() {propertyValues {Curve {keys [{interpMode=RCIM_Cubic, leaveTangent=-3.0846, tangentMode=RCTM_Break, time=0, value=1}, <at 1 0>]}}}
     Scale Curve=15
    )
   )
  ]
  rendererClasses [&NiagaraMeshRendererProperties, &NiagaraLightRendererProperties]
  renderers [
   {rendererClass=&NiagaraMeshRendererProperties, rendererIndex=0}
   {rendererClass=&NiagaraLightRendererProperties, rendererIndex=1}
  ]
 }
 Minimal {
  emitterUpdateScript [
   &EmitterState()
   &SpawnBurst_Instantaneous(
    Spawn Count=&MakeCustomIntFromBool(Bool=$User.HitWeakSpot, True INT=1, False INT=0)
    Use Spawn Probability=false
    UseLoopCountLimit=false
   )
  ]
  particleSpawnScript [
   &InitializeParticle(
    Lifetime=0.8
    Sprite Rotation Angle=&RandomRangeFloat(Minimum=15, Maximum=-15)
    Sprite Rotation Mode=::NewEnumerator0
    Sprite Size Mode=::NewEnumerator3
    Uniform Sprite Size=&Multiply_Float(A=85, B=&FN_FrontEndFloatModifier(Gameplay=1, Front End=$User.FrontEndSize))
    UsePositionOffset=false
    Write Material Random=true
   )
   &SubUVAnimation(
    Direct Frame Index:NiagaraFloat=&SelectFloatFromArray(
     Direct Array Index=&MakeIntFromFloat(Float=$User.VariantID)
     Float Selection Array=DataInterface() {propertyValues {FloatData [0,2,3,1]}}
    )
    "Sprite Renderer "=DataInterface() {propertyValues {SpriteRenderer=&self:Minimal.NiagaraSpriteRendererProperties_0}}
    SubUV Animation Mode:ENiagara_SubUVLookupModeV2=::NewEnumerator5
   ) {enabled=false}
   /SetVariables(Particles.SubImage:NiagaraFloat=&SelectFloatFromArray(
     Direct Array Index=&MakeIntFromFloat(Float=$User.VariantID)
     Float Selection Array=DataInterface() {propertyValues {FloatData [0,2,3,1]}}
    )) {isSetParameters=true}
  ]
  particleUpdateScript [
   &ParticleState()
   &ScaleSpriteSize(Uniform Curve Sprite Scale=DataInterface() {propertyValues {Curve {keys [{tangentMode=RCTM_User, time=0.0033, value=0.3}, <at 0.1758 0.8516>, {tangentMode=RCTM_User, time=1.0076, value=1.1475}]}}})
   &ScaleColor(
    Color Scale:LinearColor=&MakeCustomLinearColorFromBool(
     Bool=&SetBoolByFloatComparison(A=$User.VariantID, B=0, Comparison Type=::NewEnumerator4)
     True Color=&ColorFromCurve(ColorCurve=DataInterface() {propertyValues {
        AlphaCurve {keys [<at 0.2047 1 RCIM_Cubic>,<at 1 0 RCIM_Cubic>]}
        BlueCurve {keys [
          <at 0 0.094>
          {arriveTangent=-0.2441, interpMode=RCIM_Cubic, leaveTangent=-0.2441, time=0.0614, value=0}
          <at 0.1979 0>
          <at 0.3853 0>
          <at 0.7204 0.0082 RCIM_Cubic>
         ]}
        GreenCurve {keys [
          <at 0 0>
          {arriveTangent=0.1517, interpMode=RCIM_Cubic, leaveTangent=0.1517, time=0.0614, value=8.6231}
          <at 0.1979 0.6009>
          <at 0.3853 0.0585>
          <at 0.7204 0 RCIM_Cubic>
         ]}
        RedCurve {keys [<at 0 1>,<at 0.0614 12 RCIM_Cubic>,<at 0.1979 1>,<at 0.3853 1>,<at 0.7204 0.2066 RCIM_Cubic>]}
       }})
     False Color=&MakeCustomLinearColorFromBool(
      Bool=&SetBoolByFloatComparison(A=$User.VariantID, B=1, Comparison Type=::NewEnumerator4)
      True Color=&ColorFromCurve(ColorCurve=DataInterface() {propertyValues {
         AlphaCurve {keys [<at 0.2226 1 RCIM_Cubic>,<at 1 0 RCIM_Cubic>]}
         BlueCurve {keys [
           <at 0 0.0062>
           {arriveTangent=0.042, interpMode=RCIM_Cubic, leaveTangent=0.042, time=0.0614, value=2.2238}
           <at 0.2405 0.0335>
           <at 0.3853 0.0224>
           <at 0.7204 0.0273 RCIM_Cubic>
          ]}
         GreenCurve {keys [
           <at 0 1>
           {arriveTangent=-1.3925, interpMode=RCIM_Cubic, leaveTangent=-1.3925, time=0.0614, value=12}
           <at 0.2405 1>
           <at 0.3853 0.4635>
           <at 0.7204 0.1198 RCIM_Cubic>
          ]}
         RedCurve {keys [
           <at 0 0.0466>
           {arriveTangent=-0.1205, interpMode=RCIM_Cubic, leaveTangent=-0.1205, time=0.0614, value=5.4782}
           <at 0.2405 0.2945>
           <at 0.3853 0.0002>
           <at 0.7204 0 RCIM_Cubic>
          ]}
        }})
      False Color=&MakeCustomLinearColorFromBool(
       Bool=&SetBoolByFloatComparison(A=$User.VariantID, B=2, Comparison Type=::NewEnumerator4)
       True Color=&ColorFromCurve(ColorCurve=DataInterface() {propertyValues {
          AlphaCurve {keys [<at 0.1954 1 RCIM_Cubic>,<at 1 0 RCIM_Cubic>]}
          BlueCurve {keys [<at 0 1>,<at 0.0614 9.9073 RCIM_Cubic>,<at 0.2616 1>,<at 0.3853 1>,<at 0.7204 0.1441 RCIM_Cubic>]}
          GreenCurve {keys [
            <at 0 0.2796>
            {arriveTangent=-0.1598, interpMode=RCIM_Cubic, leaveTangent=-0.1598, time=0.0614, value=12}
            <at 0.2616 0.8276>
            <at 0.3853 0.2181>
            <at 0.7204 0.0067 RCIM_Cubic>
           ]}
          RedCurve {keys [<at 0 0>,<at 0.0614 0.5882 RCIM_Cubic>,<at 0.2616 0.1605>,<at 0.3853 0>,<at 0.7204 0.0099 RCIM_Cubic>]}
         }})
       False Color=&ColorFromCurve(ColorCurve=DataInterface() {propertyValues {
          AlphaCurve {keys [<at 0.227 1 RCIM_Cubic>,<at 1 0 RCIM_Cubic>]}
          BlueCurve {keys [<at 0 0>,<at 0.0614 0 RCIM_Cubic>,<at 0.2637 0.0962>,<at 0.3853 0>,<at 0.7204 0.0027 RCIM_Cubic>]}
          GreenCurve {keys [
            <at 0 0.8801>
            {arriveTangent=-1.0221, interpMode=RCIM_Cubic, leaveTangent=-1.0221, time=0.0614, value=8.6231}
            <at 0.2637 0.8598>
            <at 0.3853 0.4863>
            <at 0.7204 0.0344 RCIM_Cubic>
           ]}
          RedCurve {keys [<at 0 1>,<at 0.0614 12 RCIM_Cubic>,<at 0.2637 1>,<at 0.3853 1>,<at 0.7204 0.0434 RCIM_Cubic>]}
         }})
      )
     )
    )
    Scale Mode=3
   )
   &DynamicMaterialParameters(
    Index 0 Param 2=&FloatFromCurve(FloatCurve=DataInterface() {propertyValues {Curve {
        defaultValue=1
        keys [
         {arriveTangent=0.2253, interpMode=RCIM_Cubic, leaveTangent=0.2253, tangentMode=RCTM_User, time=0.2, value=0}
         {arriveTangent=3.578, interpMode=RCIM_Cubic, leaveTangent=3.578, tangentMode=RCTM_User, time=1.009, value=1}
        ]
       }}})
    Module.Index0Float0Write=true
    Module.Index0Float1Write=true
    Module.Index0Float2Write=true
    Module.Index0Float3Write=true
   )
   /SetVariables(Particles.DarkColor:LinearColor=&MultiplyLinearColor(Linear Color=$Particles.Color, Scale Factor=<0.06 0.06 0.1 1.2>)) {isSetParameters=true}
   /SetVariables(
    Particles.DarkDynPar:Vector4f=<2 0.1 0 0>
    Particles.DarkSize:Vector2f=&Multiply_Vector2D_ByFloat(Float=2.6, Vector2D=$Particles.SpriteSize)
   ) {isSetParameters=true}
  ]
  rendererClasses [&NiagaraSpriteRendererProperties]
  renderers [
   {rendererClass=&NiagaraSpriteRendererProperties}
   {rendererClass=&NiagaraSpriteRendererProperties, rendererIndex=1}
  ]
 }
 Particle {
  emitterUpdateScript [
   &EmitterState(
    Enable Distance Culling=true
    Inactive Response=::NewEnumerator1
    Life Cycle Mode=::NewEnumerator1
    Loop Behavior=::NewEnumerator1
    Loop Duration=0.5
    Max Distance Response=::KillImmediately
    Module.bApplySpawnCountScaleByDistance=true
    Module.bUseMaxDistance=true
    Module.bUseMinDistance=false
    Scalability Mode=::NewEnumerator1
    Scale Spawn Count=true
    Spawn Count Scale By Distance=DataInterface() {propertyValues {Curve {keys [<at 0.3676 0.9657>,<at 0.9962 0.4982>]}}}
    UseLoopDelay=false
   )
   &SpawnBurst_Instantaneous(
    Spawn Count=&UniformRangedInt(Minimum=13, Maximum=16)
    Use Spawn Probability=false
    UseLoopCountLimit=false
   )
  ]
  particleSpawnScript [
   &InitializeParticle(
    Color=&SelectLinearColorFromArray(
     Array Sampling Mode=::NewEnumerator1
     Direct Array Index=&MakeIntFromFloat(Float=$User.VariantID)
     Color Selection Array=DataInterface() {propertyValues {ColorData [<rgba 2 1.1584 0 1>, <rgba 0.0967 2 0.0065 1>, <rgba 0 0.7199 2 1>, <rgba 1 0.9198 0 1>]}}
     Random Seed=?
    )
    Lifetime Max=1
    Lifetime Min=0.3
    Lifetime Mode=::NewEnumerator1
    Sprite Size Mode=::NewEnumerator3
    Uniform Sprite Size=&Multiply_Float(A=&RandomRangeFloat(Minimum=4, Maximum=6.5), B=1)
    UsePositionOffset=false
    Write Material Random=true
   )
   &SphereLocation(
    MaskSpawnGroup=false
    Override Seed=false
    Sphere Radius=5
    Surface Only=false
    UseNonUniformScale=false
    UseOffset=false
    UseSphereOrientation=false
   ) {module=SphereLocation001}
   &AddVelocityFromPoint(
    Module.Use Falloff Distance=false
    Velocity Strength=&Multiply_Float(
     A=&RandomRangeFloat(Minimum=400, Maximum=800)
     B=&MakeCustomFloatFromBool(Bool=$User.bIsFrontEndPreview, True Float=$User.FrontEndSize, False Float=1)
    )
   )
  ]
  particleUpdateScript [
   &ParticleState()
   &Drag(Drag=&UniformRangedFloat(Minimum=5, Maximum=10), Use Linear Drag=true, Use Rotational Drag=false)
   &GravityForce(Gravity=&RandomRangeVector(Minimum=<0 0 -100>, Maximum=<0 0 -300>))
   &ScaleColor(
    Scale Alpha=&FloatFromCurve(FloatCurve=DataInterface() {propertyValues {Curve {keys [<at 0.4 1.0113 RCIM_Cubic>,<at 1.0011 0.002>]}}})
    ScaleA=true
    ScaleRGB=true
   )
   &CurlNoiseForce(Noise Strength=&RandomRangeFloat(Minimum=200, Maximum=400), Pan Noise=true)
   &SolveForcesAndVelocity(Clamp Velocity=false, Limit Acceleration=false)
   &ScaleSpriteSize(Scale Factor=&Vector2DFromCurve(
     Vector2Curve=DataInterface() {propertyValues {
       XCurve {keys [
         {arriveTangent=5.9728, interpMode=RCIM_Cubic, leaveTangent=5.9728, tangentMode=RCTM_User, time=0, value=0.1}
         {arriveTangent=-0.1, interpMode=RCIM_Cubic, leaveTangent=-0.1, tangentMode=RCTM_User, time=0.18, value=0.8}
         {interpMode=RCIM_Cubic, tangentMode=RCTM_User, time=1, value=0}
        ]}
       YCurve {keys [
         {arriveTangent=6.2715, interpMode=RCIM_Cubic, leaveTangent=6.2715, tangentMode=RCTM_User, time=0, value=0.1}
         {arriveTangent=-0.1001, interpMode=RCIM_Cubic, leaveTangent=-0.1001, tangentMode=RCTM_User, time=0.18, value=0.8}
         {interpMode=RCIM_Cubic, tangentMode=RCTM_User, time=0.9988, value=0}
        ]}
      }}
     ScaleCurve=&Vector2DFromFloat(Value=&MakeCustomFloatFromBool(Bool=$User.bIsFrontEndPreview, True Float=$User.FrontEndSize, False Float=1))
    ))
   &ScaleSpriteSizeBySpeed(Max Scale Factor=<0.5 6>, Min Scale Factor=<0.8 1>)
  ]
  rendererClasses [&NiagaraSpriteRendererProperties]
  renderers [{rendererClass=&NiagaraSpriteRendererProperties}]
 }
 Particle001 {
  emitterUpdateScript [
   &EmitterState(
    Enable Distance Culling=true
    Inactive Response=::NewEnumerator1
    Life Cycle Mode=::NewEnumerator1
    Loop Behavior=::NewEnumerator1
    Loop Duration=0.5
    Max Distance Response=::KillImmediately
    Module.bApplySpawnCountScaleByDistance=true
    Module.bUseMaxDistance=true
    Module.bUseMinDistance=false
    Scalability Mode=::NewEnumerator1
    Scale Spawn Count=true
    Spawn Count Scale By Distance=DataInterface() {propertyValues {Curve {keys [<at 0.3676 0.9657>,<at 0.9962 0.4982>]}}}
    UseLoopDelay=false
   )
   &SpawnBurst_Instantaneous(
    Spawn Count=&UniformRangedInt(Minimum=13, Maximum=16)
    Use Spawn Probability=false
    UseLoopCountLimit=false
   )
  ]
  particleSpawnScript [
   &InitializeParticle(
    Color=<rgba 0.2726 0.0882 0.0223 1>
    Lifetime Max=0.5
    Lifetime Min=0.25
    Lifetime Mode=::NewEnumerator1
    Mesh Scale Mode=::NewEnumerator1
    Mesh Uniform Scale Max=0.55
    Mesh Uniform Scale Min=0.2
    UsePositionOffset=false
    Write Material Random=true
   )
   &SphereLocation(
    MaskSpawnGroup=false
    Override Seed=false
    Sphere Radius=5
    Surface Only=false
    UseNonUniformScale=false
    UseOffset=false
    UseSphereOrientation=false
   ) {module=SphereLocation001}
   &AddVelocityFromPoint(
    Module.Use Falloff Distance=false
    Velocity Strength=&Multiply_Float(
     A=&RandomRangeFloat(Minimum=400, Maximum=800)
     B=&MakeCustomFloatFromBool(Bool=$User.bIsFrontEndPreview, True Float=$User.FrontEndSize, False Float=1)
    )
   )
   &InitialMeshOrientation(Use Rotation Vector=false)
  ]
  particleUpdateScript [
   &ParticleState()
   &ScaleMeshSize(Scale Factor=&VectorFromFloat(Value=&FloatFromCurve(FloatCurve=DataInterface() {propertyValues {Curve {
         defaultValue=1
         keys [
          {arriveTangent=9.0725, interpMode=RCIM_Cubic, leaveTangent=9.0725, tangentMode=RCTM_User, time=0, value=0}
          {arriveTangent=0.66, interpMode=RCIM_Cubic, leaveTangent=0.66, tangentMode=RCTM_User, time=0.1408, value=0.9847}
          {arriveTangent=-1.866, interpMode=RCIM_Cubic, leaveTangent=-1.866, tangentMode=RCTM_User, time=1, value=0}
         ]
        }}})))
   &Drag(Drag=&UniformRangedFloat(Minimum=5, Maximum=10), Use Linear Drag=true, Use Rotational Drag=false)
   &GravityForce(Gravity=&RandomRangeVector(Minimum=<0 0 -500>, Maximum=<0 0 -980>))
   &ScaleColor(
    Scale Alpha=&FloatFromCurve(FloatCurve=DataInterface() {propertyValues {Curve {keys [<at 0.4 1.0113 RCIM_Cubic>,<at 1.0011 0.002>]}}})
    ScaleA=true
    ScaleRGB=true
   )
   &CurlNoiseForce(Noise Strength=&RandomRangeFloat(Minimum=200, Maximum=400), Pan Noise=true)
   &SolveForcesAndVelocity(Clamp Velocity=false, Limit Acceleration=false)
   &ScaleSpriteSize(Scale Factor=&Vector2DFromCurve(
     Vector2Curve=DataInterface() {propertyValues {
       XCurve {keys [
         {arriveTangent=5.9728, interpMode=RCIM_Cubic, leaveTangent=5.9728, tangentMode=RCTM_User, time=0, value=0.1}
         {arriveTangent=-0.1, interpMode=RCIM_Cubic, leaveTangent=-0.1, tangentMode=RCTM_User, time=0.18, value=0.8}
         {interpMode=RCIM_Cubic, tangentMode=RCTM_User, time=1, value=0}
        ]}
       YCurve {keys [
         {arriveTangent=6.2715, interpMode=RCIM_Cubic, leaveTangent=6.2715, tangentMode=RCTM_User, time=0, value=0.1}
         {arriveTangent=-0.1001, interpMode=RCIM_Cubic, leaveTangent=-0.1001, tangentMode=RCTM_User, time=0.18, value=0.8}
         {interpMode=RCIM_Cubic, tangentMode=RCTM_User, time=0.9988, value=0}
        ]}
      }}
     ScaleCurve=&Vector2DFromFloat(Value=&MakeCustomFloatFromBool(Bool=$User.bIsFrontEndPreview, True Float=$User.FrontEndSize, False Float=1))
    ))
  ]
  rendererClasses [&NiagaraSpriteRendererProperties, &NiagaraMeshRendererProperties]
  renderers [
   {rendererClass=&NiagaraSpriteRendererProperties}
   {rendererClass=&NiagaraMeshRendererProperties, rendererIndex=1}
  ]
 }
 Smoke {
  emitterUpdateScript [
   &EmitterState(Life Cycle Mode=::NewEnumerator1, Loop Behavior=::NewEnumerator1, Loop Duration=0.1, UseLoopDelay=false)
   &SpawnBurst_Instantaneous(Spawn Count=3, Use Spawn Probability=false)
  ]
  particleSpawnScript [
   &InitializeParticle(
    Color=&SelectLinearColorFromArray(
     Array Sampling Mode=::NewEnumerator0
     Direct Array Index=?
     Color Selection Array=DataInterface() {propertyValues {ColorData [<rgba 0.0365 0.0092 0.0009 2>, <rgba 0.2 0.0452 0.0011 1>]}}
     Random Seed=0
    )
    Lifetime Max=0.6
    Lifetime Min=0.4
    Lifetime Mode=::NewEnumerator1
    Sprite Rotation Mode=::NewEnumerator2
    Sprite Size Mode=::NewEnumerator1
    Uniform Sprite Size Max=90
    Uniform Sprite Size Min=70
    UsePositionOffset=false
    Write Material Random=true
   ) {module=InitializeParticle001}
   &SphereLocation(
    MaskSpawnGroup=false
    Override Seed=true
    Random Seed=20
    Sphere Radius=5
    Surface Only=false
    UseNonUniformScale=false
    UseOffset=false
    UseSphereOrientation=false
   )
   &CameraOffset(Camera Offset Amount=20) {module=CameraOffset001}
   &AddVelocity(
    Module.Use Falloff Distance=false
    Override Seed=false
    Velocity Mode=::NewEnumerator1
    Velocity Speed=&MakeCustomFloatFromBool(Bool=$User.bIsFrontEndPreview, True Float=&Multiply_Float(A=100, B=$User.FrontEndSize), False Float=100)
   )
  ]
  particleUpdateScript [
   &ParticleState()
   &SpriteRotationRate(Rotation Rate=&Multiply_Float(
     A=&RandomRangeFloat(Minimum=-1, Maximum=1)
     B=&FloatFromCurve(
      FloatCurve=DataInterface() {propertyValues {Curve {keys [
          {arriveTangent=-4.0107, interpMode=RCIM_Cubic, leaveTangent=-4.0107, tangentMode=RCTM_User, time=-0.0046, value=1.0113}
          {arriveTangent=-0.2496, interpMode=RCIM_Cubic, leaveTangent=-0.2496, tangentMode=RCTM_User, time=0.5, value=0.16}
         ]}}}
      CurveIndex=$Emitter.Age
      Scale Curve=150
     )
    ))
   &Drag(Drag=&RandomRangeFloat(Minimum=3, Maximum=5), Use Linear Drag=true, Use Rotational Drag=false)
   &GravityForce(Gravity=&RandomRangeVector(Minimum=<0 0 -50>, Maximum=<0 0 -100>))
   &SolveForcesAndVelocity(Clamp Velocity=false, Limit Acceleration=false)
   &ScaleColor(
    Scale Alpha=&FloatFromCurve(
     FloatCurve=DataInterface() {propertyValues {Curve {keys [<at 0.1219 0>,<at 0.2 1>,<at 0.75 1>,<at 1 0>]}}}
     Scale Curve=&MakeCustomFloatFromBool(Bool=$User.bIsFrontEndPreview, True Float=0.3, False Float=0.8)
    )
    Scale RGB=&MakeVectorFromLinearColorRGB(Color=&ColorFromCurve(ColorCurve=DataInterface() {propertyValues {
        AlphaCurve {keys [<at 0 1 RCIM_Cubic>]}
        BlueCurve {keys [{arriveTangent=-1.5, interpMode=RCIM_Cubic, time=0.0052, value=1}, <at 1 0.5 RCIM_Cubic>]}
        GreenCurve {keys [{arriveTangent=-1.5, interpMode=RCIM_Cubic, time=0.0052, value=1}, <at 1 0.5 RCIM_Cubic>]}
        RedCurve {keys [{arriveTangent=-1.5, interpMode=RCIM_Cubic, time=0.0052, value=1}, <at 1 0.5 RCIM_Cubic>]}
       }}))
    ScaleA=true
    ScaleRGB=true
   )
   &ScaleSpriteSize(Scale Factor=&Vector2DFromFloat(Value=&FloatFromCurve(
      FloatCurve=DataInterface() {propertyValues {Curve {keys [
          {arriveTangent=3.8824, interpMode=RCIM_Cubic, leaveTangent=3.8824, tangentMode=RCTM_User, time=0, value=0.15}
          {arriveTangent=0.9544, interpMode=RCIM_Cubic, leaveTangent=0.9544, tangentMode=RCTM_User, time=0.1915, value=0.721}
          {arriveTangent=-0.0529, interpMode=RCIM_Cubic, leaveTangent=-0.0529, tangentMode=RCTM_User, time=1, value=1}
         ]}}}
      Scale Curve=&MakeCustomFloatFromBool(Bool=$User.bIsFrontEndPreview, True Float=$User.FrontEndSize, False Float=1)
     )))
   &DynamicMaterialParameters(
    Index 0 Param 1=&Add_Float(
     A=$Particles.RandomFloat
     B=&FloatFromCurve(
      FloatCurve=DataInterface() {propertyValues {Curve {keys [
          {arriveTangent=2.7533, interpMode=RCIM_Cubic, leaveTangent=1.8958, tangentMode=RCTM_User, time=0, value=1}
          {arriveTangent=0.8292, interpMode=RCIM_Cubic, leaveTangent=0.8292, tangentMode=RCTM_User, time=0.321, value=1.3453}
          {arriveTangent=0.0529, interpMode=RCIM_Cubic, leaveTangent=0.2021, tangentMode=RCTM_User, time=1, value=1.5}
         ]}}}
      CurveIndex=&Add_Float(A=&Multiply_Float(A=$Particles.Initial.MaterialRandom, B=0.5), B=$Particles.NormalizedAge)
     )
    )
    Index 0 Param 2=&FloatFromCurve(FloatCurve=DataInterface() {propertyValues {Curve {keys [
         {arriveTangent=0.0453, interpMode=RCIM_Cubic, tangentMode=RCTM_Break, time=0, value=-1}
         {arriveTangent=0.0604, interpMode=RCIM_Cubic, leaveTangent=0.0604, tangentMode=RCTM_User, time=1, value=-1}
        ]}}})
    Index 0 Param 3=&FloatFromCurve(FloatCurve=DataInterface() {propertyValues {Curve {keys [
         {arriveTangent=-53.9199, interpMode=RCIM_Cubic, leaveTangent=-8.9727, tangentMode=RCTM_User, time=0, value=1}
         {arriveTangent=0.2347, interpMode=RCIM_Cubic, leaveTangent=0.2723, tangentMode=RCTM_Break, time=0.25, value=0}
        ]}}})
    Index 0 Param 4=&FloatFromCurve(FloatCurve=DataInterface() {propertyValues {Curve {keys [<at 0.5 1>,<at 1 0>]}}})
    Index 2 Param 1=&FloatFromCurve(FloatCurve=DataInterface() {propertyValues {Curve {keys [
         {interpMode=RCIM_Cubic, leaveTangent=4.2485, tangentMode=RCTM_Break, time=0, value=0}
         {arriveTangent=1.0548, interpMode=RCIM_Cubic, leaveTangent=1.0548, tangentMode=RCTM_User, time=0.17, value=0.5715}
         <at 1 1>
        ]}}})
    Index 2 Param 2=&FloatFromCurve(
     FloatCurve=DataInterface() {propertyValues {Curve {keys [
         {arriveTangent=0.9316, interpMode=RCIM_Cubic, leaveTangent=0.9316, tangentMode=RCTM_User, time=0, value=0.1}
         {arriveTangent=0.0259, interpMode=RCIM_Cubic, leaveTangent=0.0259, tangentMode=RCTM_User, time=1, value=0.4}
        ]}}}
     Scale Curve=&RandomRangeFloat(Minimum=0.2, Maximum=0.7)
    )
    Index 2 Param 3=&FloatFromCurve(FloatCurve=DataInterface() {propertyValues {Curve {keys [
         {interpMode=RCIM_Cubic, leaveTangent=2.3064, tangentMode=RCTM_Break, time=0, value=1}
         {interpMode=RCIM_Cubic, leaveTangent=11.1233, tangentMode=RCTM_Break, time=1, value=2}
        ]}}})
    Index 2 Param 4=&FloatFromCurve(
     FloatCurve=DataInterface() {propertyValues {Curve {keys [
         {arriveTangent=-0.6867, interpMode=RCIM_Cubic, leaveTangent=-0.6867, tangentMode=RCTM_User, time=0, value=1}
         {interpMode=RCIM_Cubic, leaveTangent=-1.2956, tangentMode=RCTM_Break, time=1, value=0.75}
        ]}}}
     Scale Curve=1.6
    )
    Module.Param0WriteEnabled=true
    Module.Param0WriteEnabled2=true
    Module.Param1WriteEnabled=false
    Module.Param1WriteEnabled2=true
    Module.Param2WriteEnabled=true
    Module.Param2WriteEnabled2=true
    Module.Param3WriteEnabled=true
    Module.Param3WriteEnabled2=true
    Write Parameter Index 2=true
   )
  ]
  rendererClasses [&NiagaraSpriteRendererProperties]
  renderers [{rendererClass=&NiagaraSpriteRendererProperties}]
 }
 Sparks_Toon {
  emitterUpdateScript [
   &EmitterState(Life Cycle Mode=::NewEnumerator1, Loop Behavior=::NewEnumerator1, Loop Delay=0.005, Loop Duration=0.25, UseLoopDelay=false)
   &SpawnRate(SpawnRate=15, Use Spawn Probability=false)
   &SpawnBurst_Instantaneous(Spawn Count=5, Use Spawn Probability=false, UseLoopCountLimit=false)
  ]
  particleSpawnScript [
   &FN_SetSpawnGroup(Number of Spawn Groups=2)
   &InitializeParticle(
    Color=<rgba 0.05 0.05 0.05 0.5>
    Lifetime Max=0.7
    Lifetime Min=0.3
    Lifetime Mode=::NewEnumerator1
    Sprite Size Mode=::NewEnumerator1
    Uniform Sprite Size Max=15
    Uniform Sprite Size Min=10
    UsePositionOffset=false
    Write Material Random=true
   )
   &SphereLocation(
    MaskSpawnGroup=false
    Override Seed=false
    Sphere Radius=&Multiply_Float(A=50, B=&FN_FrontEndFloatModifier(Gameplay=1, Front End=$System.FrontendScale))
    Surface Expansion Mode=::NewEnumerator1
    Surface Only=true
    UseNonUniformScale=false
    UseOffset=true
    UseSphereOrientation=false
   )
   &CameraOffset(Camera Offset Amount=1)
   /SetVariables(Particles.VisibilityTag=&SpawnGroup()) {isSetParameters=true}
  ]
  particleUpdateScript [
   &ParticleState()
   &SolveForcesAndVelocity(Clamp Velocity=false, Limit Acceleration=false)
   &Color(
    Color=&MakeCustomLinearColorFromBool(
     Bool=&SetBoolByFloatComparison(A=$User.VariantID, B=0, Comparison Type=::NewEnumerator4)
     True Color=&ColorFromCurve(ColorCurve=DataInterface() {propertyValues {
        AlphaCurve {keys [<at 0.0065 0>,<at 0.3936 0.5 RCIM_Cubic>,<at 1 0 RCIM_Cubic>]}
        BlueCurve {keys [<at 0 0 RCIM_Cubic>,<at 0.7105 0.0141 RCIM_Cubic>]}
        GreenCurve {keys [<at 0 0.75 RCIM_Cubic>,<at 0.7105 0.0163 RCIM_Cubic>]}
        RedCurve {keys [<at 0 3 RCIM_Cubic>,<at 0.7105 0.3594 RCIM_Cubic>]}
       }})
     False Color=&MakeCustomLinearColorFromBool(
      Bool=&SetBoolByFloatComparison(A=$User.VariantID, B=1, Comparison Type=::NewEnumerator4)
      True Color=&ColorFromCurve(ColorCurve=DataInterface() {propertyValues {
         AlphaCurve {keys [<at 0.0065 0>,<at 0.3936 0.5 RCIM_Cubic>,<at 1 0 RCIM_Cubic>]}
         BlueCurve {keys [<at 0 0 RCIM_Cubic>,<at 0.7105 0.0177 RCIM_Cubic>]}
         GreenCurve {keys [<at 0 3 RCIM_Cubic>,<at 0.7105 0.3594 RCIM_Cubic>]}
         RedCurve {keys [<at 0 0.7571 RCIM_Cubic>,<at 0.7105 0.0484 RCIM_Cubic>]}
        }})
      False Color=&MakeCustomLinearColorFromBool(
       Bool=&SetBoolByFloatComparison(A=$User.VariantID, B=2, Comparison Type=::NewEnumerator4)
       True Color=&ColorFromCurve(ColorCurve=DataInterface() {propertyValues {
          AlphaCurve {keys [<at 0.0065 0>,<at 0.3936 0.5 RCIM_Cubic>,<at 1 0 RCIM_Cubic>]}
          BlueCurve {keys [<at 0 3 RCIM_Cubic>,<at 0.7105 0.3594 RCIM_Cubic>]}
          GreenCurve {keys [<at 0 2.7891 RCIM_Cubic>,<at 0.7105 0.1002 RCIM_Cubic>]}
          RedCurve {keys [<at 0 0.7163 RCIM_Cubic>,<at 0.7105 0.0039 RCIM_Cubic>]}
         }})
       False Color=&ColorFromCurve(ColorCurve=DataInterface() {propertyValues {
          AlphaCurve {keys [<at 0.0065 0>,<at 0.3936 0.5 RCIM_Cubic>,<at 1 0 RCIM_Cubic>]}
          BlueCurve {keys [<at 0 0 RCIM_Cubic>,<at 0.7105 0.0055 RCIM_Cubic>]}
          GreenCurve {keys [<at 0 2.6987 RCIM_Cubic>,<at 0.7105 0.2402 RCIM_Cubic>]}
          RedCurve {keys [<at 0 3 RCIM_Cubic>,<at 0.7105 0.3594 RCIM_Cubic>]}
         }})
      )
     )
    )
    Scale Alpha=&FloatFromCurve(
     FloatCurve=DataInterface() {propertyValues {Curve {keys [
         {interpMode=RCIM_Cubic, tangentMode=RCTM_User, time=-0.0063, value=0.0184}
         {arriveTangent=-0.8483, interpMode=RCIM_Cubic, leaveTangent=-0.8483, tangentMode=RCTM_User, time=0.1718, value=1.7147}
         {arriveTangent=-5.0978, interpMode=RCIM_Cubic, leaveTangent=-5.0978, tangentMode=RCTM_User, time=0.2818, value=1.2143}
         {arriveTangent=-0.2783, interpMode=RCIM_Cubic, leaveTangent=-0.2783, tangentMode=RCTM_User, time=0.4, value=0.7}
         {arriveTangent=0.3112, interpMode=RCIM_Cubic, leaveTangent=0.3112, tangentMode=RCTM_User, time=0.5407, value=1.3419}
         {arriveTangent=-3.7119, interpMode=RCIM_Cubic, leaveTangent=-3.7119, tangentMode=RCTM_User, time=0.7242, value=0.3749}
         {interpMode=RCIM_Cubic, tangentMode=RCTM_User, time=1.0011, value=0.002}
        ]}}}
     Scale Curve=0.8
    )
    Scale Color=&VectorFromFloat(Value=50)
   )
   &ScaleSpriteSize(Scale Factor=&Vector2DFromFloat(Value=&FloatFromCurve(
      FloatCurve=DataInterface() {propertyValues {Curve {keys [
          {arriveTangent=4.1302, interpMode=RCIM_Cubic, leaveTangent=4.1302, tangentMode=RCTM_User, time=0, value=0}
          {interpMode=RCIM_Cubic, tangentMode=RCTM_User, time=0.5, value=1}
          {tangentMode=RCTM_User, time=1, value=0}
         ]}}}
      Scale Curve=&Multiply_Float(
       A=&MaskFloatBySpawnGroup(Masked Float=1, Passthrough Float=1.5)
       B=&FN_FrontEndFloatModifier(Gameplay=1, Front End=$User.FrontEndScale)
      )
     )))
  ]
  rendererClasses [&NiagaraSpriteRendererProperties]
  renderers [
   {rendererClass=&NiagaraSpriteRendererProperties}
   {rendererClass=&NiagaraSpriteRendererProperties, rendererIndex=1}
  ]
 }
 glow {
  emitterUpdateScript [
   &EmitterState(Inactive Response=::NewEnumerator1, Life Cycle Mode=::NewEnumerator1, Loop Behavior=::NewEnumerator1, Loop Duration=0.5, UseLoopDelay=false)
   &SpawnBurst_Instantaneous(Use Spawn Probability=false, UseLoopCountLimit=false)
  ]
  particleSpawnScript [
   &InitializeParticle(
    Color=<rgba 0.1354 0.1354 0.1354 2>
    Lifetime=0.4
    Sprite Size Mode=::NewEnumerator3
    Uniform Sprite Size=230
    UsePositionOffset=false
    Write Material Random=true
   )
   &CameraOffset(Camera Offset Amount=&Add_Float(A=10, B=&FN_PickFloatDependOnPlatform(TargetPlatformActive=1, TargetPlatformNoActive=100)))
  ]
  particleUpdateScript [
   &ParticleState()
   &ScaleSpriteSize(
    Uniform Curve Scale=&MakeCustomFloatFromBool(Bool=$User.bIsFrontEndPreview, True Float=$User.FrontEndSize, False Float=1)
    Uniform Curve Sprite Scale=DataInterface() {propertyValues {Curve {keys [<at 0 1>, {tangentMode=RCTM_User, time=1, value=0.2}]}}}
   )
   /SetVariables(Particles.Size002:Vector2f=&Multiply_Vector2D(A=$Particles.SpriteSize, B=&Vector2DFromFloat(Value=0.4))) {isSetParameters=true}
   &ScaleColor(
    Scale Alpha=&FloatFromCurve(
     FloatCurve=DataInterface() {propertyValues {Curve {keys [
         {arriveTangent=4.4475, interpMode=RCIM_Cubic, leaveTangent=4.4475, tangentMode=RCTM_User, time=0, value=0.6}
         {arriveTangent=5.3518, interpMode=RCIM_Cubic, leaveTangent=-2.8479, tangentMode=RCTM_Break, time=0.12, value=1.198}
         {arriveTangent=-1.9582, interpMode=RCIM_Cubic, leaveTangent=-1.9582, tangentMode=RCTM_User, time=0.4276, value=0.5658}
         <at 1 0 RCIM_Cubic>
        ]}}}
     Scale Curve=&MakeCustomFloatFromBool(Bool=$User.bIsFrontEndPreview, True Float=$User.FrontEndSize, False Float=1)
    )
    Scale RGB=&MakeVectorFromLinearColorRGB(Color=&MakeCustomLinearColorFromBool(
      Bool=&SetBoolByFloatComparison(A=$User.VariantID, B=0, Comparison Type=::NewEnumerator4)
      True Color=&ColorFromCurve(ColorCurve=DataInterface() {propertyValues {
         AlphaCurve {keys []}
         BlueCurve {keys [
           <at 0.0215 0>
           {arriveTangent=0.008, interpMode=RCIM_Cubic, leaveTangent=0.008, time=0.2355, value=0}
           {arriveTangent=0.0107, interpMode=RCIM_Cubic, leaveTangent=0.0107, time=0.6615, value=0}
           <at 0.8659 0.0068>
          ]}
         GreenCurve {keys [
           <at 0.0215 0.4459>
           {arriveTangent=-0.5281, interpMode=RCIM_Cubic, leaveTangent=-0.5281, time=0.2355, value=0.0609}
           {arriveTangent=-0.0966, interpMode=RCIM_Cubic, leaveTangent=-0.0966, time=0.6615, value=0.0073}
           <at 0.8659 0>
          ]}
         RedCurve {keys [
           <at 0.0215 0.9974>
           {arriveTangent=-1.027, interpMode=RCIM_Cubic, leaveTangent=-1.027, time=0.2355, value=0.9974}
           {arriveTangent=-1.3755, interpMode=RCIM_Cubic, leaveTangent=-1.3755, time=0.6615, value=0.2726}
           <at 0.8659 0.1302>
          ]}
        }})
      False Color=&MakeCustomLinearColorFromBool(
       Bool=&SetBoolByFloatComparison(A=$User.VariantID, B=1, Comparison Type=::NewEnumerator4)
       True Color=&ColorFromCurve(ColorCurve=DataInterface() {propertyValues {
          AlphaCurve {keys []}
          BlueCurve {keys [
            <at 0.0272 0.0514>
            {arriveTangent=-0.0637, interpMode=RCIM_Cubic, leaveTangent=-0.0637, time=0.1323, value=0.2188}
            {arriveTangent=-0.2493, interpMode=RCIM_Cubic, leaveTangent=-0.2493, time=0.5734, value=0.0166}
            <at 0.8972 0.0281>
           ]}
          GreenCurve {keys [
            <at 0.0272 1.5>
            {arriveTangent=-2.3934, interpMode=RCIM_Cubic, leaveTangent=-2.3934, time=0.1323, value=1.5}
            {arriveTangent=-1.8816, interpMode=RCIM_Cubic, leaveTangent=-1.8816, time=0.5734, value=0.1927}
            <at 0.8972 0.0608>
           ]}
          RedCurve {keys [
            <at 0.0272 1.115>
            {arriveTangent=-2.0112, interpMode=RCIM_Cubic, leaveTangent=-2.0112, time=0.1323, value=0.5886}
            {arriveTangent=-0.7695, interpMode=RCIM_Cubic, leaveTangent=-0.7695, time=0.5734, value=0.0164}
            <at 0.8972 0>
           ]}
         }})
       False Color=&MakeCustomLinearColorFromBool(
        Bool=&SetBoolByFloatComparison(A=$User.VariantID, B=2, Comparison Type=::NewEnumerator4)
        True Color=&ColorFromCurve(ColorCurve=DataInterface() {propertyValues {
           AlphaCurve {keys []}
           BlueCurve {keys [
             <at 0.0272 1.4393>
             {arriveTangent=-2.2822, interpMode=RCIM_Cubic, leaveTangent=-2.2822, time=0.1323, value=1.5}
             {arriveTangent=-1.8816, interpMode=RCIM_Cubic, leaveTangent=-1.8816, time=0.5734, value=0.1927}
             <at 0.8972 0.0608>
            ]}
           GreenCurve {keys [
             <at 0.0272 1.5>
             {arriveTangent=-2.63, interpMode=RCIM_Cubic, leaveTangent=-2.63, time=0.1323, value=0.9655}
             {arriveTangent=-1.2622, interpMode=RCIM_Cubic, leaveTangent=-1.2622, time=0.5734, value=0.0634}
             <at 0.8972 0>
            ]}
           RedCurve {keys [
             <at 0.0272 0.8175>
             {arriveTangent=-1.461, interpMode=RCIM_Cubic, leaveTangent=-1.461, time=0.1323, value=0.1858}
             {arriveTangent=-0.2395, interpMode=RCIM_Cubic, leaveTangent=-0.2395, time=0.5734, value=0.0195}
             <at 0.8972 0.0027>
            ]}
          }})
        False Color=&ColorFromCurve(ColorCurve=DataInterface() {propertyValues {
           AlphaCurve {keys []}
           BlueCurve {keys [
             <at 0.0272 0.5023>
             {arriveTangent=-0.8478, interpMode=RCIM_Cubic, leaveTangent=-0.8478, time=0.1323, value=0}
             <at 0.5734 0.0392 RCIM_Cubic>
             <at 0.8972 0>
            ]}
           GreenCurve {keys [
             <at 0.0272 1.5>
             {arriveTangent=-2.4381, interpMode=RCIM_Cubic, leaveTangent=-2.4381, time=0.1323, value=1.3357}
             {arriveTangent=-1.7463, interpMode=RCIM_Cubic, leaveTangent=-1.7463, time=0.5734, value=0.1683}
             <at 0.8972 0>
            ]}
           RedCurve {keys [
             <at 0.0272 1.4982>
             {arriveTangent=-2.3901, interpMode=RCIM_Cubic, leaveTangent=-2.3901, time=0.1323, value=1.5}
             {arriveTangent=-1.9611, interpMode=RCIM_Cubic, leaveTangent=-1.9611, time=0.5734, value=0.1927}
             <at 0.8972 0>
            ]}
          }})
       )
      )
     ))
    ScaleA=true
    ScaleRGB=true
   )
   /SetVariables(Particles.color002:LinearColor=&MakeLinearColorFromVectorAndFloat(
     "Vector (RGB)"=&MakeVectorFromLinearColorRGB(Color=&SelectLinearColorFromArray(
       Array Sampling Mode=::NewEnumerator1
       Direct Array Index=&MakeIntFromFloat(Float=$User.VariantID)
       Color Selection Array=DataInterface() {propertyValues {ColorData [<rgba 5 2.59 0 1>, <rgba 0 5 0.116 1>, <rgba 0 2.5813 5 1>, <rgba 0.9685 1 0 1>]}}
       Random Seed=?
      ))
     "Float (Alpha)"=&FloatFromCurve(FloatCurve=DataInterface() {propertyValues {Curve {keys [{interpMode=RCIM_Cubic, leaveTangent=-6.869, tangentMode=RCTM_Break, time=0, value=1}, <at 0.45 0>]}}})
    )) {isSetParameters=true}
  ]
  rendererClasses [&NiagaraSpriteRendererProperties]
  renderers [
   {rendererClass=&NiagaraSpriteRendererProperties}
   {rendererClass=&NiagaraSpriteRendererProperties, rendererIndex=1}
  ]
 }
 spine_Center {
  emitterSpawnScript [/SetVariables(
    Emitter.DarkCol:LinearColor=&SelectLinearColorFromArray(
     Array Sampling Mode=::NewEnumerator1
     Direct Array Index=&MakeIntFromFloat(Float=$User.VariantID)
     Color Selection Array=DataInterface() {propertyValues {ColorData [<rgba 10 0.7892 0 1>, <rgba 1.0444 10 0 1>, <rgba 0.0483 4.6666 10 1>, <rgba 10 7.9635 0 1>]}}
     Random Seed=?
    )
    Emitter.LightCol:LinearColor=&SelectLinearColorFromArray(
     Array Sampling Mode=::NewEnumerator1
     Direct Array Index=&MakeIntFromFloat(Float=$User.VariantID)
     Color Selection Array=DataInterface() {propertyValues {ColorData [<rgba 1 0.5744 0.0454 1>, <rgba 0.3469 1 0.0547 1>, <rgba 0.3278 1 0.9103 1>, <rgba 1 0.9231 0.1988 1>]}}
     Random Seed=?
    )
   ) {isSetParameters=true}]
  emitterUpdateScript [
   &EmitterState(Inactive Response=::NewEnumerator1, Life Cycle Mode=::NewEnumerator1, Loop Behavior=::NewEnumerator1, Loop Duration=0.5, UseLoopDelay=false)
   &SpawnBurst_Instantaneous(Use Spawn Probability=false, UseLoopCountLimit=false)
  ]
  particleSpawnScript [
   &InitializeParticle(
    Color=<rgba 1 1 1 0.6>
    Lifetime=0.15
    Mesh Scale Max=<1.55 1.55 1.55>
    Mesh Scale Min=<1.2 1.2 1.2>
    Mesh Scale Mode=::NewEnumerator2
    UsePositionOffset=false
    Write Material Random=true
   )
   &InitialMeshOrientation(
    Mesh Orientation Mode=::NewEnumerator2
    Rotation=&RandomRangeVector(Minimum=<-1 -1 -1>, Maximum=<1 1 1>)
    Use Rotation Vector=true
   )
   &CameraOffset(Camera Offset Amount=10)
  ]
  particleUpdateScript [
   &ParticleState()
   &ScaleMeshSize(Scale Factor=&VectorFromFloat(Value=&FloatFromCurve(
      FloatCurve=DataInterface() {propertyValues {Curve {keys [
          {arriveTangent=3.9833, interpMode=RCIM_Cubic, leaveTangent=3.9833, tangentMode=RCTM_User, time=0, value=0.2}
          {arriveTangent=5.6067, tangentMode=RCTM_User, time=0.15, value=1}
          {arriveTangent=-0.5799, interpMode=RCIM_Cubic, leaveTangent=-4.8336, tangentMode=RCTM_Break, time=0.5592, value=0.9}
          {arriveTangent=0.0139, interpMode=RCIM_Cubic, leaveTangent=0.0139, tangentMode=RCTM_User, time=1, value=0}
         ]}}}
      Scale Curve=&MakeCustomFloatFromBool(Bool=$User.bIsFrontEndPreview, True Float=$User.FrontEndSize, False Float=1)
     )))
   &DynamicMaterialParameters(
    Index 0 Param 1=&FloatFromCurve(FloatCurve=DataInterface() {propertyValues {Curve {keys [<at 0.7 0>,<at 1 1>]}}})
    Index 0 Param 2=1
    Index 0 Param 3=1
    Index 0 Param 4=1
    Module.Param0WriteEnabled=true
    Module.Param1WriteEnabled=true
    Module.Param2WriteEnabled=true
    Module.Param3WriteEnabled=true
   )
   &ScaleColor(
    Scale RGB=&MakeVectorFromLinearColorRGB(Color=&ColorFromCurve(ColorCurve=DataInterface() {propertyValues {
        AlphaCurve {keys [<at 0 1 RCIM_Cubic>,<at 0.8581 0 RCIM_Cubic>]}
        BlueCurve {keys [<at 0.154 5 RCIM_Cubic>,<at 0.5025 1 RCIM_Cubic>]}
        GreenCurve {keys [<at 0.154 5 RCIM_Cubic>,<at 0.5025 1 RCIM_Cubic>]}
        RedCurve {keys [<at 0.154 5 RCIM_Cubic>,<at 0.5025 1 RCIM_Cubic>]}
       }}))
    ScaleA=true
    ScaleRGB=true
   )
  ]
  rendererClasses [&NiagaraMeshRendererProperties]
  renderers [{rendererClass=&NiagaraMeshRendererProperties, rendererIndex=0}]
 }
 spine_sharp {
  emitterUpdateScript [
   &EmitterState(Inactive Response=::NewEnumerator1, Life Cycle Mode=::NewEnumerator1, Loop Behavior=::NewEnumerator1, Loop Duration=0.5, UseLoopDelay=false)
   &SpawnBurst_Instantaneous(Spawn Count=4, Use Spawn Probability=false)
  ]
  particleSpawnScript [
   &InitializeParticle(
    Color=&SelectLinearColorFromArray(
     Array Sampling Mode=::NewEnumerator1
     Direct Array Index=&MakeIntFromFloat(Float=$User.VariantID)
     Color Selection Array=DataInterface() {propertyValues {ColorData [<rgba 1 0.3818 0 1>, <rgba 0.1406 1 0.0854 1>, <rgba 0 1.3897 1.5 1>, <rgba 1 0.8275 0.0108 1>]}}
     Random Seed=?
    )
    Lifetime Max=0.24
    Lifetime Min=0.18
    Lifetime Mode=::NewEnumerator1
    Sprite Size Max=<15 45>
    Sprite Size Min=<8 40>
    Sprite Size Mode=::NewEnumerator2
    UsePositionOffset=false
    Write Material Random=true
   )
   &AddVelocity(Module.Use Falloff Distance=false, Override Seed=false, Velocity Mode=::NewEnumerator1, Velocity Speed=12)
   &CameraOffset(Camera Offset Amount=10)
  ]
  particleUpdateScript [
   &ParticleState()
   &ScaleSpriteSize(
    Uniform Curve Scale=&Multiply_Float(
     A=0.6
     B=&MakeCustomFloatFromBool(Bool=$User.bIsFrontEndPreview, True Float=$User.FrontEndSize, False Float=1)
    )
    Uniform Curve Sprite Scale=DataInterface() {propertyValues {Curve {keys [
        {arriveTangent=4.0942, interpMode=RCIM_Cubic, leaveTangent=4.0942, tangentMode=RCTM_User, time=0, value=0.5}
        {arriveTangent=0.967, interpMode=RCIM_Cubic, leaveTangent=0.967, tangentMode=RCTM_User, time=0.2796, value=1.55}
        {interpMode=RCIM_Cubic, leaveTangent=0.2272, time=1, value=2}
       ]}}}
   )
   &SolveForcesAndVelocity(Clamp Velocity=false, Limit Acceleration=false)
   &ScaleColor(
    Scale Alpha=&FloatFromCurve(
     FloatCurve=DataInterface() {propertyValues {Curve {keys [<at 0 1>,<at 1 0>]}}}
     Scale Curve=&MakeCustomFloatFromBool(Bool=$User.bIsFrontEndPreview, True Float=1, False Float=2)
    )
    Scale RGB=&MakeVectorFromLinearColorRGB(Color=&ColorFromCurve(ColorCurve=DataInterface() {propertyValues {
        AlphaCurve {keys []}
        BlueCurve {keys [<at 0 1.5 RCIM_Cubic>,<at 0.1256 1 RCIM_Cubic>]}
        GreenCurve {keys [<at 0 1.1746 RCIM_Cubic>,<at 0.1256 1 RCIM_Cubic>]}
        RedCurve {keys [<at 0 0.7802 RCIM_Cubic>,<at 0.1256 1 RCIM_Cubic>]}
       }}))
    ScaleA=true
    ScaleRGB=true
   )
  ]
  rendererClasses [&NiagaraSpriteRendererProperties]
  renderers [{rendererClass=&NiagaraSpriteRendererProperties}]
 }
 spine_sharp001 {
  emitterUpdateScript [
   &EmitterState(Inactive Response=::NewEnumerator1, Life Cycle Mode=::NewEnumerator1, Loop Behavior=::NewEnumerator1, Loop Duration=0.5, UseLoopDelay=false)
   &SpawnBurst_Instantaneous(Spawn Count=4, Use Spawn Probability=false)
  ]
  particleSpawnScript [
   &InitializeParticle(
    Color=&MakeCustomLinearColorFromBool(
     Bool=&SetBoolByFloatComparison(A=$User.VariantID, B=0, Comparison Type=::NewEnumerator4)
     True Color=&SelectLinearColorFromArray(
      Array Sampling Mode=::NewEnumerator0
      Direct Array Index=?
      Color Selection Array=DataInterface() {propertyValues {ColorData [<rgba 0.1458 0 0.0018 0.6>, <rgba 0.217 0.0191 0 1>]}}
      Random Seed=0
     )
     False Color=&MakeCustomLinearColorFromBool(
      Bool=&SetBoolByFloatComparison(A=$User.VariantID, B=1, Comparison Type=::NewEnumerator4)
      True Color=&SelectLinearColorFromArray(
       Array Sampling Mode=::NewEnumerator0
       Direct Array Index=?
       Color Selection Array=DataInterface() {propertyValues {ColorData [<rgba 0.1455 0.2812 0.0376 0.6>, <rgba 0.0371 0.0885 0.0214 1>]}}
       Random Seed=0
      )
      False Color=&MakeCustomLinearColorFromBool(
       Bool=&SetBoolByFloatComparison(A=$User.VariantID, B=2, Comparison Type=::NewEnumerator4)
       True Color=&SelectLinearColorFromArray(
        Array Sampling Mode=::NewEnumerator0
        Direct Array Index=?
        Color Selection Array=DataInterface() {propertyValues {ColorData [<rgba 0.0277 0.219 0.2812 0.6>, <rgba 0 0.0207 0.0885 1>]}}
        Random Seed=0
       )
       False Color=&SelectLinearColorFromArray(
        Array Sampling Mode=::NewEnumerator0
        Direct Array Index=?
        Color Selection Array=DataInterface() {propertyValues {ColorData [<rgba 0.2812 0.2582 0 0.6>, <rgba 0.0885 0.0551 0 1>]}}
        Random Seed=0
       )
      )
     )
    )
    Lifetime Max=0.24
    Lifetime Min=0.18
    Lifetime Mode=::NewEnumerator1
    Sprite Size Max=<20 55>
    Sprite Size Min=<10 45>
    Sprite Size Mode=::NewEnumerator2
    UsePositionOffset=false
    Write Material Random=true
   )
   &AddVelocity(Module.Use Falloff Distance=false, Override Seed=false, Velocity Mode=::NewEnumerator1, Velocity Speed=12)
   &CameraOffset(Camera Offset Amount=10)
  ]
  particleUpdateScript [
   &ParticleState()
   &ScaleSpriteSize(
    Uniform Curve Scale=&Multiply_Float(
     A=0.6
     B=&MakeCustomFloatFromBool(Bool=$User.bIsFrontEndPreview, True Float=$User.FrontEndSize, False Float=1)
    )
    Uniform Curve Sprite Scale=DataInterface() {propertyValues {Curve {keys [
        {arriveTangent=4.0942, interpMode=RCIM_Cubic, leaveTangent=4.0942, tangentMode=RCTM_User, time=0, value=0.5}
        {arriveTangent=0.967, interpMode=RCIM_Cubic, leaveTangent=0.967, tangentMode=RCTM_User, time=0.2796, value=1.55}
        {interpMode=RCIM_Cubic, leaveTangent=0.2272, time=1, value=2}
       ]}}}
   )
   &SolveForcesAndVelocity(Clamp Velocity=false, Limit Acceleration=false)
   &ScaleColor(
    Scale Alpha=&FloatFromCurve(
     FloatCurve=DataInterface() {propertyValues {Curve {keys [<at 0 1>,<at 1 0>]}}}
     Scale Curve=&MakeCustomFloatFromBool(Bool=$User.bIsFrontEndPreview, True Float=1, False Float=2)
    )
    Scale RGB=&MakeVectorFromLinearColorRGB(Color=&ColorFromCurve(ColorCurve=DataInterface() {propertyValues {
        AlphaCurve {keys []}
        BlueCurve {keys [<at 0 0.6 RCIM_Cubic>,<at 0.1256 0.4 RCIM_Cubic>]}
        GreenCurve {keys [<at 0 0.5241 RCIM_Cubic>,<at 0.1256 0.4 RCIM_Cubic>]}
        RedCurve {keys [<at 0 0.4648 RCIM_Cubic>,<at 0.1256 0.4 RCIM_Cubic>]}
       }}))
    ScaleA=true
    ScaleRGB=true
   )
  ]
  rendererClasses [&NiagaraSpriteRendererProperties]
  renderers [{rendererClass=&NiagaraSpriteRendererProperties}]
 }
}
summary {
 systemName=NS_Pickaxe_DroveRay_Impact
 emitters [
  {emitterName=Smoke, rendererClasses [&NiagaraSpriteRendererProperties]}
  {emitterName=Particle, rendererClasses [&NiagaraSpriteRendererProperties]}
  {emitterName=spine_Center, rendererClasses [&NiagaraMeshRendererProperties]}
  {emitterName=spine_sharp, rendererClasses [&NiagaraSpriteRendererProperties]}
  {emitterName=spine_sharp001, rendererClasses [&NiagaraSpriteRendererProperties]}
  {emitterName=glow, rendererClasses [&NiagaraSpriteRendererProperties]}
  {emitterName=Empty, rendererClasses [&NiagaraMeshRendererProperties, &NiagaraLightRendererProperties]}
  {emitterName=Sparks_Toon, rendererClasses [&NiagaraSpriteRendererProperties]}
  {emitterName=Particle001, rendererClasses [&NiagaraSpriteRendererProperties, &NiagaraMeshRendererProperties]}
  {emitterName=Minimal, rendererClasses [&NiagaraSpriteRendererProperties]}
 ]
 userVariables [
  User.bIsFrontEnd {type=NiagaraBool}
  User.bIsFrontEndPreview {type=NiagaraBool}
  User.FrontEndScale {type=NiagaraFloat, defaultValue=0.4, description=""}
  User.FrontEndSize {type=NiagaraFloat, defaultValue=0.4, description=""}
  User.HitWeakSpot {type=NiagaraBool}
  User.VariantID {type=NiagaraFloat, defaultValue=0, description=""}
 ]
}
