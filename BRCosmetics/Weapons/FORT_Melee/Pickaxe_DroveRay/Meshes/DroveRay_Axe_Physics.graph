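# Physics asset for the DroveRay two-handed pickaxe mesh, defining bodies for the handle and a dangling crest charm.
# Two capsules cover the handle, small capsules form four chain links, and a larger capsule is the charm at the end.
# Constraints link handle, numbered bones 01 to 04, then the charm in series, allowing moderate swing with almost no twist.
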PhysicsAsset /BRCosmetics/Weapons/FORT_Melee/Pickaxe_DroveRay/Meshes/DroveRay_Axe_Physics
bodies {
 Melee_TwoHanded_Handle [
  {shapeType=Capsule, length=70, radius=5, shapeName=Melee_TwoHanded_Handle_capsule, center=<0 0 20>}
  {shapeType=Capsule, length=8, radius=4, shapeName=Melee_TwoHanded_Handle_capsule, center=<0 -0.9283 -33.3583>}
 ]
 dyn_crest_charm [{
   shapeType=Capsule
   length=4
   radius=6.5
   shapeName=dyn_crest_charm_01_capsule
   center=<8.3352 -0.0004 0.0014>
   rotation=<rot 0 89.9972 -89.9211>
  }]
 dyn_crest_charm_01 [{shapeType=Capsule, length=2, radius=1, shapeName=dyn_crest_charm_01_capsule, center=<1 0 -0.0087>, rotation=<rot 0 89.9972 -89.9211>}]
 dyn_crest_charm_02 [{shapeType=Capsule, length=2, radius=1, shapeName=dyn_crest_charm_01_capsule, center=<1 0 -0.0087>, rotation=<rot 0 89.9972 -89.9211>}]
 dyn_crest_charm_03 [{shapeType=Capsule, length=2, radius=1, shapeName=dyn_crest_charm_01_capsule, center=<1 0 -0.0087>, rotation=<rot 0 89.9972 -89.9211>}]
 dyn_crest_charm_04 [{shapeType=Capsule, length=2, radius=1, shapeName=dyn_crest_charm_01_capsule, center=<1 0 -0.0087>, rotation=<rot 0 89.9972 -89.9211>}]
}
constraints [
 {bone1Name=dyn_crest_charm_01, bone2Name=Melee_TwoHanded_Handle, swing1LimitDegrees=15, swing2LimitDegrees=25, twistLimitDegrees=1}
 {bone1Name=dyn_crest_charm_02, bone2Name=dyn_crest_charm_01, swing1LimitDegrees=25, swing2LimitDegrees=35, twistLimitDegrees=1}
 {bone1Name=dyn_crest_charm_03, bone2Name=dyn_crest_charm_02, swing1LimitDegrees=25, swing2LimitDegrees=25, twistLimitDegrees=1}
 {bone1Name=dyn_crest_charm_04, bone2Name=dyn_crest_charm_03, swing1LimitDegrees=25, swing2LimitDegrees=25, twistLimitDegrees=1}
 {bone1Name=dyn_crest_charm, bone2Name=dyn_crest_charm_04, swing1LimitDegrees=25, swing2LimitDegrees=25, twistLimitDegrees=1}
]
